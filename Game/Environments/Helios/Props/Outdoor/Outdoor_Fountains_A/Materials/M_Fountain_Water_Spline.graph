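# Translucent, refractive material for fountain water jets drawn on spline strips, thickened through SplineThicken.
# One panned water normal map drives surface normal, strip width and normal-directed vertex displacement, with vertex colors varying tiling, speed, width and optionally displacement.
# Caustics sampled on the same UVs shape opacity and, multiplied with streaks, the emissive tint.
# Edge gradient, depth fade and a Fountain_Water_Progress sphere mask along V limit where water shows.
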Material /Game/Environments/Helios/Props/Outdoor/Outdoor_Fountains_A/Materials/M_Fountain_Water_Spline
graph {
 properties {
  bAutomaticallySetUsageInEditor=false
  blendMode=BLEND_Translucent
  refractionMethod=RM_PixelNormalOffset
  translucencyLightingMode=TLM_VolumetricPerVertexDirectional
  usedShadingModels=Default Lit
 }
 nodes [
  n1 = ScalarParameter() {defaultValue=1, parameterName=TilingX}
  n2 = ScalarParameter() {defaultValue=1, parameterName=TilingY}
  n3 = Multiply(
   A=LinearInterpolate(
    A=ScalarParameter() {parameterName=Tile_Offset_Min}
    B=ScalarParameter() {defaultValue=1, parameterName=Tile_Offset_Max}
    Alpha=VertexColor().G
   )
   B=Multiply(A=TextureCoordinate(), B=AppendVector(A=@n1, B=@n2))
  )
  n4 = ComponentMask(None=@n3) {r=true}
  n5 = Add(
   A=ComponentMask(None=@n3) {g=true}
   B=LinearInterpolate(A=ScalarParameter() {parameterName=OffsetY_Min}, B=ScalarParameter() {defaultValue=1, parameterName=OffsetY_Max}, Alpha=VertexColor().G)
  )
  n6 = VectorParameter() {parameterName=Tiling_Normal, defaultValue=<rgba 1 1 0 1>}
  n7 = VectorParameter() {parameterName=Speed, defaultValue=<rgba 1 1 0 1>}
  n8 = ComponentMask(None=@n7.RGB) {g=true}
  n9 = LinearInterpolate(A=ScalarParameter() {parameterName=Speed_OffsetY_Min}, B=ScalarParameter() {parameterName=Speed_OffsetY_Max}, Alpha=VertexColor().G)
  n10 = Panner(
   Coordinate=Multiply(A=AppendVector(A=@n4, B=@n5), B=ComponentMask(None=@n6.RGB) {g=true, r=true})
   Speed=AppendVector(A=ComponentMask(None=@n7.RGB) {r=true}, B=Add(A=@n8, B=@n9))
  )
  n11 = TextureSample(UVs=@n10) {samplerType=SAMPLERTYPE_Normal, texture=/FortWater/Textures/T_Water_N}
  n12 = Multiply(A=@n11.G, B=ScalarParameter() {defaultValue=1, parameterName=WidthTextureAdd})
  n13 = LinearInterpolate(
   A=ScalarParameter() {defaultValue=10, parameterName=WidthBase}
   B=ScalarParameter() {defaultValue=10, parameterName=WidthBaseMax}
   Alpha=VertexColor().B
  )
  n14 = ScalarParameter() {defaultValue=10, parameterName="WidthTip (S)"}
  n15 = VectorParameter() {parameterName=WPO_Mult, defaultValue=<rgba 1 1 1 1>}
  n16 = StaticSwitchParameter(True=Multiply(A=@n15.RGB, B=VertexColor().R), False=@n15.RGB) {parameterName=Use_RandomWPO?}
  Add(
   A=MaterialFunctionCall(WidthBase=Add(A=@n12, B=@n13), WidthTip=Add(A=@n12, B=@n14)) {materialFunction=&SplineThicken}.WorldPositionOffset
   B=Multiply(A=VertexNormalWS(), B=Multiply(A=@n11.RGB, B=@n16))
  )
  AppendVector()
  n17 = MaterialFunctionCall(In=TextureSample(UVs=@n10) {texture=/FortWater/Textures/T_Caustics_3layer}.G, Contrast=ScalarParameter() {parameterName="Contrast (S)"}) {materialFunction=&CheapContrast}
  n18 = MaterialFunctionCall(UV Channel=TextureCoordinate()) {materialFunction=&LinearGradient}
  n19 = OneMinus(None=@n18.UGradient)
  n20 = Multiply(A=Multiply(A=@n17.Result, B=Multiply(A=@n18.UGradient, B=@n19)), B=ScalarParameter() {defaultValue=1, parameterName=FinalOpacity}) {constB=5}
  n21 = Constant()
  n22 = Add(A=ScalarParameter() {defaultValue=1, parameterName=Fountain_Water_Progress}) {constB=-0.5}
  n23 = Saturate(None=SphereMask(A=TextureCoordinate() {uTiling=0}, B=AppendVector(A=@n21, B=@n22)) {attenuationRadius=0.5, hardnessPercent=25})
  LinearInterpolate(B=ScalarParameter() {parameterName=Refrac}, Alpha=Multiply(A=@n20, B=@n23)) {constA=1}
  n24 = LinearInterpolate(
   A=ScalarParameter() {defaultValue=1, parameterName=Emmisve_TileX}
   B=ScalarParameter() {defaultValue=1, parameterName=Emmisve_TileX_Max}
   Alpha=VertexColor().B
  )
  n25 = LinearInterpolate(
   A=ScalarParameter() {defaultValue=1, parameterName=Emmisve_TileY}
   B=ScalarParameter() {defaultValue=1, parameterName=Emmisve_TileY_Max}
   Alpha=VertexColor().B
  )
  n26 = TextureSample(UVs=Multiply(A=AppendVector(A=@n24, B=@n25), B=@n10)) {samplerType=SAMPLERTYPE_LinearColor, texture=/Game/Effects/Fort_Effects/Textures/T_Streaks_01_M}
  LinearInterpolate(
   B=VectorParameter() {parameterName=EmmisiveColor, defaultValue=<rgba 1 0.3657 0.0249 1>}.RGB
   Alpha=MaterialFunctionCall(In=Multiply(A=@n26.B, B=@n17.Result), Contrast=ScalarParameter() {parameterName=Contrast_Emmisive}) {materialFunction=&CheapContrast}.Result
  ) {constB=0}
  MaterialFunctionCall(Normal=@n11.RGB, Flatness=ScalarParameter() {defaultValue=1, parameterName="Flatness (S)"}) {materialFunction=&FlattenNormal}
  Multiply(A=DepthFade(Opacity=@n20), B=@n23)
  ScalarParameter() {defaultValue=1, parameterName=WPO}
  ScalarParameter() {parameterName=Roughness}
  ScalarParameter() {parameterName=Param_1}
  ScalarParameter() {parameterName=Specular}
  ScalarParameter() {parameterName=Metallic}
  TextureSample() {samplerType=SAMPLERTYPE_Masks, texture=/FortWater/Textures/OceanNoise__Mask}
  VectorParameter() {parameterName=Color, defaultValue=<rgba 0 0 0 1>}
 ]
}
